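# Physics asset for the Pizza Pit pickaxe, a Fortnite melee cosmetic with a jointed pizza cutter part.
# A long, thin capsule on b_AR_Root covers the shaft, and a zero-length capsule on dyn_pizzaCutter acts as a sphere around the cutter.
# One constraint joins the cutter to the root, locking the first swing axis and leaving twist free so the cutter can rotate.
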PhysicsAsset /BRCosmetics/Weapons/FORT_Melee/Pickaxe_Pizza_Pit/Meshes/Pickaxe_Pizza_Pit_Physics
bodies {
 b_AR_Root [{shapeType=Capsule, length=119.255, radius=2.505, center=<5.9686 0 -9.4948>, rotation=<rot 0 -89.9999 0.5551>}]
 dyn_pizzaCutter [{shapeType=Capsule, length=0, radius=15, center=<0 0.1 0>}]
}
constraints [{
  bone1Name=dyn_pizzaCutter
  bone2Name=b_AR_Root
  swing1LimitDegrees=0
  swing1Motion=Locked
  swing2LimitDegrees=10
  twistLimitDegrees=45
  twistMotion=Free
 }]
