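# Animation Blueprint for the Pizza Pit pickaxe cosmetic that adds physics-driven secondary motion to the weapon's skeleton.
# The input pose passes through a RigidBody node using the Pickaxe_Pizza_Pit_Physics asset with overridden gravity and clamped component acceleration, then goes to the output.
# An AnimDynamics node for dyn_pizzaCutter and a Copy Pose From Mesh node are left unconnected.
# On initialization the event graph caches the owner skeletal mesh in MeshToCopy, which only the unused copy node reads.
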AnimBlueprint /BRCosmetics/Weapons/FORT_Melee/Pickaxe_Pizza_Pit/Meshes/Pickaxe_Pizza_Pit_Skeleton_AnimBlueprint
target_skeleton=&/Pickaxe_Pizza_Pit_Skeleton
graphs [
 AnimGraph {nodes [
   Anim Dynamics - Bone: dyn_pizzaCutter(Alpha=1, ExternalForce="0,0,0") {
    angularSpringConstant=100
    bGravityOverrideInSimSpace=true
    simulationSpace=World
    alphaBoolBlend {blendOption=Linear}
    boundBone {boneName=dyn_pizzaCutter}
    physicsBodyDefinitions [{
      sphereCollisionRadius=0
      boundBone {boneName=dyn_pizzaCutter}
      boxExtents=<5 35 30>
      constraintSetup {
       twistAxis=AxisY
       angularLimitsMax=<180 0 0>
       angularLimitsMin=<-180 0 0>
       angularTarget=<0 0 0>
       linearAxesMax=<1 0 0>
       linearAxesMin=<-1 0 0>
      }
      localJointOffset=<0 0 0>
     }]
   }
   Copy Pose From Mesh(SourceMeshComponent=Get MeshToCopy() -> (MeshToCopy:Skeletal Mesh Component Object Reference)) {bUseAttachedParent=true}
   AnimGraphNode_SubInput_0 = Input Pose()
   n1 = RigidBody(ComponentPose=Local To Component(LocalPose=@AnimGraphNode_SubInput_0)) {
    bOverrideWorldGravity=true
    overridePhysicsAsset=&/Pickaxe_Pizza_Pit_Physics
    alphaBoolBlend {blendOption=Linear}
    componentAppliedLinearAccClamp=<300 300 300>
    componentLinearAccScale=<1 1 1>
    componentLinearVelScale=<1 1 1>
    overrideWorldGravity {z=-980}
   }
   Output Pose(Result=Component To Local(ComponentPose=@n1)) {layerGroup=None}
  ]}
 EventGraph {nodes [
   Event BlueprintInitializeAnimation()
   > Set MeshToCopy(MeshToCopy:Skeletal Mesh Component Object Reference=GetOwnerSkeletalMeshComponent()) -> (Output_Get:Skeletal Mesh Component Object Reference)
   TryGetPawnOwner()
   Event BlueprintUpdateAnimation()
  ]}
]
variables [MeshToCopy {type=SkeletalMeshComponent}]
